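# First-person firing sound cue for the pistol-caliber assault rifle.
# A Random node picks one of two equally weighted close-fire waves, with repeats allowed, and feeds a Modulator that adds slight pitch variation at constant volume.
# It references first-person attenuation, sound class and gunfire concurrency assets, with overrides for no spatialization and an owner-limited voice cap that stops the oldest.
# The debug flag is left enabled.
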SoundCue /Game/Sounds/Fort_Weapons/AutoRifle/Assault_PistolCaliber/Weapon_ASLT_PistolCaliber_1P_Cue
root=Modulator_3
nodes [
 {id=Modulator_3, children [Random_1], properties {pitchMax=1.02, pitchMin=0.98, volumeMax=1, volumeMin=1}}
 {id=Random_1, children [WavePlayer_4,WavePlayer_5], properties {bRandomizeWithoutReplacement=false, weights [1,1]}}
 {id=WavePlayer_4, properties {soundWaveAssetPtr=&/JW_AR_Close_Fire_02}}
 {id=WavePlayer_5, properties {soundWaveAssetPtr=&/JW_AR_Close_Fire_03}}
]
properties {
 attenuationSettings=&Fort_Weapon_Projectile_FirstPerson_Athena_Att
 bDebug=true
 maxDistance=2097152
 soundClassObject=&Fort_WeaponRanged_Fire_1P_PSM
 volumeMultiplier=1
 attenuationOverrides {bSpatialize=false}
 concurrencyOverrides {bLimitToOwner=true, maxCount=3, resolutionRule=StopOldest}
 concurrencySet [&Weapon_Gunfire_FirstPerson_Conc]
}
