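# Sound cue for the pistol-caliber assault rifle's gunfire, layering close, medium and distant shots, with a third-person fire sound class as its default.
# Each layer picks a random wave and has its own attenuation, with slight pitch variation on medium and distant.
# Sound class overrides route close to third-person fire and both others to distant fire before nested mixers blend them.
# Concurrency sets include global and local gunfire limits.
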SoundCue /Game/Sounds/Fort_Weapons/AutoRifle/Assault_PistolCaliber/Weapon_ASLT_PistolCaliber_Cue
root=Mixer_2
nodes [
 {
  id=Attenuation_3
  children [Random_2]
  properties {attenuationSettings=/Game/Sounds/Attenuation/Fort_Weapon_Projectile_Close_Athena_Att_Scoped}
 }
 {
  id=Attenuation_4
  children [Modulator_2]
  properties {attenuationSettings=&Fort_Weapon_Projectile_Distant_Mono_Athena_Att}
 }
 {
  id=Attenuation_5
  children [Modulator_4]
  properties {attenuationSettings=&Fort_Weapon_Projectile_Medium_Mono_Athena_Att}
 }
 {id=Mixer_2, children [Modulator_3,SoundClass_4], properties {inputVolume [1,1]}}
 {id=Mixer_3, children [SoundClass_3,SoundClass_5], properties {inputVolume [0.6,1]}}
 {id=Modulator_2, children [Random_1], properties {pitchMax=1.1, pitchMin=1, volumeMax=0.3, volumeMin=0.3}}
 {id=Modulator_3, children [Mixer_3]}
 {id=Modulator_4, children [Random_0], properties {pitchMax=1.1, pitchMin=1, volumeMax=1, volumeMin=1}}
 {id=Random_0, children [WavePlayer_0,WavePlayer_1,WavePlayer_6], properties {weights [1,1,1]}}
 {id=Random_1, children [WavePlayer_7,WavePlayer_8,WavePlayer_9], properties {weights [1,1,1]}}
 {id=Random_2, children [WavePlayer_3,WavePlayer_4], properties {weights [1,1]}}
 {
  id=SoundClass_3
  children [Attenuation_3]
  properties {soundClassOverride=&Fort_WeaponRanged_Fire_3P_PSM}
 }
 {
  id=SoundClass_4
  children [Attenuation_4]
  properties {soundClassOverride=&Fort_WeaponRanged_Fire_Distant}
 }
 {
  id=SoundClass_5
  children [Attenuation_5]
  properties {soundClassOverride=&Fort_WeaponRanged_Fire_Distant}
 }
 {id=WavePlayer_0, properties {soundWaveAssetPtr=&/JW_AR_Medium_Fire_01a}}
 {id=WavePlayer_1, properties {soundWaveAssetPtr=&/JW_AR_Medium_Fire_01b}}
 {id=WavePlayer_3, properties {soundWaveAssetPtr=&/JW_AR_Close_Fire_02}}
 {id=WavePlayer_4, properties {soundWaveAssetPtr=&/JW_AR_Close_Fire_03}}
 {id=WavePlayer_6, properties {soundWaveAssetPtr=&/JW_AR_Medium_Fire_01c}}
 {id=WavePlayer_7, properties {soundWaveAssetPtr=&/JW_AR_Distant_Fire_01a}}
 {id=WavePlayer_8, properties {soundWaveAssetPtr=&/JW_AR_Distant_Fire_01b}}
 {id=WavePlayer_9, properties {soundWaveAssetPtr=&/JW_AR_Distant_Fire_01c}}
]
properties {
 bDebug=true
 maxDistance=27500
 soundClassObject=&Fort_WeaponRanged_Fire_3P_PSM
 volumeMultiplier=1.1
 concurrencyOverrides {bLimitToOwner=true, maxCount=3, resolutionRule=StopOldest}
 concurrencySet [&Weapon_Gunfire_Global_Conc, &Weapon_Gunfire_Local_Conc]
}
